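# Sound cue for impacts on the MaskMaker Spire Shard weakpoint, classed as boss-fight SFX.
# A quality-level node picks between a four-wave random pool for the first level and a shared three-wave pool for the others, reusing the SpireSpike impact waves.
# A modulator adds slight downward pitch variation.
# Playback uses the MaskMaker boss mask-hit attenuation, boosted volume, and its own concurrency limit that stops the oldest instance.
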SoundCue /SpireSharedAssets/Audio/MaskMaker_SpireShard/SC_MaskMaker_SpireShardWeakpoint_Impact
root=Modulator_0
nodes [
 {id=Modulator_0, children [QualityLevel_0], properties {pitchMax=1, pitchMin=0.98}}
 {id=QualityLevel_0, children [Random_1,Random_0,Random_0,Random_0]}
 {id=Random_0, children [WavePlayer_2,WavePlayer_3,WavePlayer_4], properties {weights [1,1,1]}}
 {id=Random_1, children [WavePlayer_0,WavePlayer_1,WavePlayer_5,WavePlayer_6], properties {weights [1,1,1,1]}}
 {id=WavePlayer_0, properties {soundWaveAssetPtr=/SpireSharedAssets/Audio/Waves/SpireSpike_Weakpoint_Impact_01}}
 {id=WavePlayer_1, properties {soundWaveAssetPtr=/SpireSharedAssets/Audio/Waves/SpireSpike_Weakpoint_Impact_02}}
 {id=WavePlayer_2, properties {soundWaveAssetPtr=&SpireSpike_Weakpoint_Impact_01}}
 {id=WavePlayer_3, properties {soundWaveAssetPtr=&SpireSpike_Weakpoint_Impact_02}}
 {id=WavePlayer_4, properties {soundWaveAssetPtr=/SpireSharedAssets/Audio/Waves/SpireSpike_Weakpoint_Impact_03}}
 {id=WavePlayer_5, properties {soundWaveAssetPtr=&SpireSpike_Weakpoint_Impact_03}}
 {id=WavePlayer_6, properties {soundWaveAssetPtr=/SpireSharedAssets/Audio/Waves/SpireSpike_Weakpoint_Impact_04}}
]
properties {
 attenuationSettings=&/ATT_MaskMakerBoss_MaskHit
 bOverrideConcurrency=true
 maxDistance=2100
 soundClassObject=&Fort_BossFight_SFX
 volumeMultiplier=1.8
 attenuationOverrides {bAttenuate=false, falloffDistance=10000, attenuationShapeExtents {x=0}}
 concurrencyOverrides {maxCount=3, resolutionRule=StopOldest, retriggerTime=0.5, voiceStealReleaseTime=0.1}
}
